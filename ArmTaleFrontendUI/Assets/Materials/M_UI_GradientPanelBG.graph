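# Translucent UI material that draws a panel background as a two-color linear gradient inset by a padding border.
# GradientMask takes the green channel of UVs scaled about center, compensated for padding, then rotated; the 0.0175 Rotator speed treats Gradient_Rotation as degrees.
# PanelMask saturates pixel distance to the nearest edge minus Panel_Padding, keeping the inset resolution-independent.
# Start and end colors are lerped, with RGB and alpha split for masking.
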Material /ArmTaleFrontendUI/Assets/Materials/M_UI_GradientPanelBG
graph {
 properties {
  bAutomaticallySetUsageInEditor=false
  bUsedWithStaticMesh=false
  blendMode=BLEND_Translucent
  materialDomain=MD_UI
  usedShadingModels=Default Lit
 }
 nodes [
  n1 = LinearInterpolate(
   A=VectorParameter() {parameterName=Gradient_StartColor, sortPriority=1, defaultValue=<rgba 0 1 0 1>}.RGBA
   B=VectorParameter() {parameterName=Gradient_EndColor, sortPriority=2, defaultValue=<rgba 1 0 0 1>}.RGBA
   Alpha=NamedRerouteUsage()
  ) {constA=1, constB=0.8}
  ComponentMask(None=@n1) {b=true, g=true, r=true}
  Multiply(A=ComponentMask(None=@n1) {a=true}, B=NamedRerouteUsage())
  n2 = MaterialFunctionCall() {materialFunction=&GetUserInterfaceUV}
  n3 = OneMinus(None=@n2.Normalized UV)
  n4 = ScalarParameter() {defaultValue=1, parameterName=Panel_Padding, sortPriority=0}
  n5 = Subtract(A=Min(A=Multiply(A=@n2.Pixel Size, B=@n2.Normalized UV), B=Multiply(A=@n2.Pixel Size, B=@n3)), B=@n4) {constB=10}
  n6 = ComponentMask(None=@n5) {r=true}
  n7 = ComponentMask(None=@n5) {g=true}
  NamedRerouteDeclaration(None=Saturate(None=Min(A=@n6, B=@n7))) {name=PanelMask, nodeColor=<rgba 0 0 0 1>}
  n8 = Divide(A=@n4, B=@n2.Pixel Size)
  n9 = MaterialFunctionCall(
   UVs=@n2.Normalized UV
   Texture Scale=Subtract(A=ScalarParameter() {defaultValue=1, parameterName=Gradient_Scale, sortPriority=4}, B=Multiply(A=@n8) {constB=2})
  ) {materialFunction=&ScaleUVsByCenter}
  n10 = ScalarParameter() {parameterName=Gradient_Rotation, sortPriority=3}
  NamedRerouteDeclaration(None=ComponentMask(None=Rotator(Coordinate=@n9.UVs, Time=@n10) {speed=0.0175}) {g=true}) {name=GradientMask, nodeColor=<rgba 0 0 0 1>}
 ]
}
