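# UI material that draws a weapon mod slot tile, with parameters for empty, equipped and highlighted states.
# An SDF box supplies tile and outline masks, and a static switch toggles the button presentation.
# Filled slots layer an icon over a rarity-colored radial gradient and optional rotated bottom strip, both responding to equipped state.
# Empty slots use a channel-selected, antialiased placeholder icon at reduced opacity, whose gradient, like the outline's, responds to highlighting.
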Material /ArmTaleFrontendUI/Assets/Materials/M_UI_ModTile
graph {
 properties {
  bAutomaticallySetUsageInEditor=false
  bUsedWithStaticMesh=false
  blendMode=BLEND_AlphaComposite
  materialDomain=MD_UI
  usedShadingModels=Default Lit
 }
 nodes [
  n1 = NamedRerouteUsage()
  n2 = LinearInterpolate(A=NamedRerouteUsage(), B=NamedRerouteUsage(), Alpha=ScalarParameter() {defaultValue=1, parameterName=IsEmpty})
  n3 = Multiply(A=NamedRerouteUsage(), B=@n2)
  Multiply(A=VertexColor().A, B=StaticSwitchParameter(True=Add(A=@n1, B=@n3), False=@n2) {defaultValue=true, parameterName=DisplayAsButton})
  n4 = VectorParameter() {parameterName=RarityColor_1, defaultValue=<rgba 0 1 0.9647 1>}
  n5 = VectorParameter() {parameterName=RarityColor_2, defaultValue=<rgba 0 0.4287 1 1>}
  n6 = MaterialFunctionCall() {materialFunction=&GetUserInterfaceUV}
  n7 = ComponentMask(None=@n6.Normalized UV) {r=true}
  n8 = ScalarParameter() {parameterName=IsEquipped}
  NamedRerouteDeclaration(None=LinearInterpolate(A=LinearInterpolate(A=@n4.RGB, B=@n5.RGB, Alpha=@n7), Alpha=@n8)) {name=BottomStripColor, nodeColor=<rgba 0 0 0 1>}
  n9 = ScalarParameter() {defaultValue=0.925, parameterName=BottomStrip_Position}
  n10 = LinearInterpolate(Alpha=@n8) {constB=0.2}
  n11 = Rotator(Coordinate=@n6.Normalized UV, Time=ScalarParameter() {defaultValue=1.5, parameterName=BottomStrip_Rotation}) {speed=0.0175}
  NamedRerouteDeclaration(None=MaterialFunctionCall(DistanceField=ComponentMask(None=@n11) {g=true}, Threshold=Subtract(A=@n9, B=@n10)) {materialFunction=&MF_UI_AntiAliasing}.Outer) {name=BottomStripMask, nodeColor=<rgba 0 0 0 1>}
  n12 = Add(
   A=MaterialFunctionCall!(UVs, Size=Constant() {r=1}, CornerRoundness, KeepAspectRatio, PixelSize) {materialFunction=/FortUILibrary/MaterialFunctions/Shapes/SDFs/MF_SDF_Box}.SDF
   B=ScalarParameter() {defaultValue=0.075, parameterName=Tile_EdgePadding}
  )
  NamedRerouteDeclaration(None=MaterialFunctionCall(DistanceField=@n12) {materialFunction=&MF_UI_AntiAliasing}.Inner) {name=TileMask, nodeColor=<rgba 0 0 0 1>}
  NamedRerouteDeclaration(None=MaterialFunctionCall!(
    DistanceField=@n12
    Threshold
    Thickness=ScalarParameter() {defaultValue=0.02, parameterName=Tile_OutlineThickness}
    Alignment=ScalarParameter() {defaultValue=1, parameterName=Tile_OutlineAlignment}
    Smoothing
   ) {materialFunction=/FortUILibrary/MaterialFunctions/Shapes/Modifiers/MF_UI_ShapeOutline}.Outline) {name=TileOutlineMask, nodeColor=<rgba 0 0 0 1>}
  n13 = ScalarParameter() {defaultValue=0.5, parameterName=EmptySlotIcon_Scale}
  n14 = ChannelMaskParameter(None=TextureSample(UVs=MaterialFunctionCall(Texture Scale=@n13) {materialFunction=&ScaleUVsByCenter}.UVs) {texture=/ArmTaleFrontendUI/Assets/Textures/T_ModSlotIcons}.RGB) {parameterName=EmptySlotIcon_Channel}
  n15 = Constant() {r=0.5}
  n16 = VectorParameter() {parameterName=EmptySlotIcon_EndColor, defaultValue=<rgba 0.2016 0.2016 0.2016 1>}
  n17 = VectorParameter() {parameterName=EmptySlotIcon_StartColor, defaultValue=<rgba 0.0262 0.0262 0.0262 1>}
  n18 = ComponentMask(None=MaterialFunctionCall() {materialFunction=&GetUserInterfaceUV}.Normalized UV) {g=true}
  n19 = LinearInterpolate(A=@n16.RGB, B=@n17.RGB, Alpha=Saturate(None=ConstantBiasScale(None=@n18) {bias=-0.25, scale=2}))
  n20 = ScalarParameter() {parameterName=Highlighted}
  NamedRerouteDeclaration(None=Multiply(
    A=MaterialFunctionCall(DistanceField=@n14.R, Threshold=@n15) {materialFunction=&MF_UI_AntiAliasing}.Outer
    B=LinearInterpolate(A=@n19, Alpha=@n20)
   )) {name=EmptySlotIcon, nodeColor=<rgba 0 0 0 1>}
  n21 = VectorParameter() {parameterName=RarityColor_3, defaultValue=<rgba 0 0.0976 0.521 1>}
  n22 = VectorParameter() {parameterName=RarityColor_2, defaultValue=<rgba 0 0.4287 1 1>}
  n23 = Subtract(
   A=MaterialFunctionCall() {materialFunction=&GetUserInterfaceUV}.Normalized UV
   B=AppendVector(
    A=ScalarParameter() {parameterName=BackgroundGradient_OffsetX}
    B=ScalarParameter() {defaultValue=-0.4, parameterName=BackgroundGradient_OffsetY}
   )
  )
  n24 = ScalarParameter() {defaultValue=2.25, parameterName=BackgroundGradient_Scale}
  n25 = MaterialFunctionCall(UV=MaterialFunctionCall(UV=MaterialFunctionCall(UVs=@n23, Texture Scale=@n24) {materialFunction=&ScaleUVsByCenter}.UVs) {materialFunction=&MF_UI_UVScaling}.Scaled To Fill) {materialFunction=&MF_UI_RadialGradient}
  n26 = ScalarParameter() {parameterName=IsEquipped}
  NamedRerouteDeclaration(None=Multiply(A=LinearInterpolate(A=@n21.RGB, B=@n22.RGB, Alpha=@n25.Smooth), B=LinearInterpolate(Alpha=@n26) {constA=0.4}) {constB=0.2}) {name=ModBackground, nodeColor=<rgba 0 0 0 1>}
  n27 = LinearInterpolate(A=@n16.RGB, B=@n17.RGB, Alpha=OneMinus(None=@n18))
  NamedRerouteDeclaration(None=Multiply(A=LinearInterpolate(A=@n27, Alpha=@n20), B=NamedRerouteUsage())) {name=TileOutline, nodeColor=<rgba 0 0 0 1>}
  n28 = NamedRerouteUsage()
  n29 = LinearInterpolate(A=@n28, B=NamedRerouteUsage(), Alpha=NamedRerouteUsage())
  n30 = TextureSample(
   UVs=MaterialFunctionCall(Texture Scale=ScalarParameter() {defaultValue=0.75, parameterName=ModIcon_Scale}) {materialFunction=&ScaleUVsByCenter}.UVs
   Tex=TextureObjectParameter() {parameterName=IconTexture, texture=/WeaponMods/UI/Icons/T_Icon_DefaultMag_64}
  ) {samplerSource=SSM_Clamp_WorldGroupSettings, texture=/Game/UI/Foundation/Textures/Icons/Athena/T-T-Icon-BR-Pepper}
  NamedRerouteDeclaration(None=LinearInterpolate(A=StaticSwitchParameter(True=@n29, False=@n28) {defaultValue=true, parameterName=ShowBottomStrip}, B=@n30.RGB, Alpha=@n30.RGB)) {name=ModIcon, nodeColor=<rgba 0 0 0 1>}
  n31 = LinearInterpolate(B=ScalarParameter() {defaultValue=0.25, parameterName=EmptySlot_Opacity}, Alpha=ScalarParameter() {defaultValue=1, parameterName=IsEmpty}) {constA=1}
  StaticSwitchParameter(True=Multiply(A=NamedRerouteUsage(), B=@n31), False=@n31) {defaultValue=true, parameterName=DisplayAsButton}
 ]
}
