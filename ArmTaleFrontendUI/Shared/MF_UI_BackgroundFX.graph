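# UI material function that renders an animated screen-space background: a dark purple vertical gradient with a flickering grid of random cell values.
# Viewport UVs, with X scaled by the view's aspect ratio, feed MF_UI_ParticleGrid from FortUILibrary, whose random seed advances with Time.
# The random cell values, scaled by opacity, color-dodge over the gradient.
# Grid tiling, flicker speed and opacity are inputs with preview values as defaults.
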MaterialFunction /ArmTaleFrontendUI/Shared/MF_UI_BackgroundFX
graph {
 properties {previewMaterialDomain=MD_UI}
 nodes [
  n1 = ViewSize()
  n2 = Divide(A=ComponentMask(None=@n1) {r=true}, B=ComponentMask(None=@n1) {g=true})
  n3 = ScreenPosition()
  n4 = ComponentMask(None=@n3.ViewportUV) {r=true}
  n5 = ComponentMask(None=@n3.ViewportUV) {g=true}
  n6 = FunctionInput() {bUsePreviewValueAsDefault=true, inputName=GridTiling, inputType=FunctionInput_Scalar, previewValue=<6 0 0 0>}
  n7 = FunctionInput() {bUsePreviewValueAsDefault=true, inputName=GridFlickerSpeed, inputType=FunctionInput_Scalar, previewValue=<0.25 0 0 0>}
  n8 = MaterialFunctionCall!(
   UV=AppendVector(A=Multiply(A=@n2, B=@n4), B=@n5)
   Global Offset
   Global Rotation
   Tiling=@n6
   Cell Offset
   Cell Scale
   Cell Rotation
   Random Seed=Multiply(A=Time(), B=@n7)
  ) {materialFunction=/FortUILibrary/MaterialFunctions/Basics/MF_UI_ParticleGrid}
  n9 = FunctionInput() {bUsePreviewValueAsDefault=true, inputName=GridOpacity, inputType=FunctionInput_Scalar, previewValue=<0.4 0 0 0>}
  n10 = Constant3Vector() {constant=<rgba 0.0103 0.006 0.0232 1>}
  n11 = Constant3Vector() {constant=<rgba 0.003 0.0027 0.004 1>}
  FunctionOutput(None=MaterialFunctionCall!(Blend=Multiply(A=@n8.Random Cell Values, B=@n9), Base=LinearInterpolate(A=@n10, B=@n11, Alpha=@n5)) {materialFunction=/Engine/Functions/Engine_MaterialFunctions03/Blends/Blend_ColorDodge}.Result)
 ]
}
