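# UI material function that outputs a scaled, rotated icon with a black drop shadow as RGB, RGBA and single channels.
# UVs pass through MF_UI_UVScaling's fill output and center scaling, then Rotators whose 0.0175 speed makes IconRotation read as degrees.
# A second texture sample, offset before rotation so the shadow direction stays fixed, supplies shadow alpha scaled by ShadowOpacity.
# Icon alpha composites the icon over that shadow in both color and alpha.
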MaterialFunction /ArmTaleFrontendUI/Shared/MF_UI_IconWithDropShadow
graph {
 properties {
  description="Takes in an icon texture and rotation/shadow parameters and returns a scaled, rotated version of the icon with a drop shadow."
  previewMaterialDomain=MD_UI
 }
 nodes [
  n1 = FunctionInput() {
   bUsePreviewValueAsDefault=true
   description=Optional UVs to apply to the icon texture.
   inputName=UVs
   inputType=FunctionInput_Vector2
   previewValue=<1 1 0 1>
  }
  n2 = FunctionInput() {
   bUsePreviewValueAsDefault=true
   description=The scale of the icon texture.
   inputName=IconScale
   inputType=FunctionInput_Vector2
   sortPriority=2
   previewValue=<1 1 0 1>
  }
  n3 = MaterialFunctionCall(
   UVs=MaterialFunctionCall(UV=@n1) {materialFunction=&MF_UI_UVScaling}.Scaled To Fill
   Texture Scale=@n2
  ) {materialFunction=&ScaleUVsByCenter}
  n4 = FunctionInput() {
   bUsePreviewValueAsDefault=true
   description=The rotation of the icon texture.
   inputName=IconRotation
   inputType=FunctionInput_Scalar
   sortPriority=3
   previewValue=<0 0 0 0>
  }
  n5 = FunctionInput() {description=The texture to use as the icon., inputName=IconTexture, inputType=FunctionInput_Texture2D, sortPriority=1}
  n6 = TextureSample(UVs=Rotator(Coordinate=@n3.UVs, Time=@n4) {speed=0.0175}, Tex=@n5) {samplerSource=SSM_Clamp_WorldGroupSettings}
  n7 = LinearInterpolate(B=@n6.RGB, Alpha=@n6.A)
  n8 = FunctionInput() {
   bUsePreviewValueAsDefault=true
   description=The opacity of the icon drop shadow.
   inputName=ShadowOpacity
   inputType=FunctionInput_Scalar
   sortPriority=5
   previewValue=<0.3 0 0 0>
  }
  n9 = FunctionInput() {
   bUsePreviewValueAsDefault=true
   description=The XY offset of the icon drop shadow.
   inputName=ShadowOffset
   inputType=FunctionInput_Vector2
   sortPriority=4
   previewValue=<0.015 0.02 0 1>
  }
  n10 = Rotator(Coordinate=Subtract(A=@n3.UVs, B=@n9), Time=@n4) {speed=0.0175}
  n11 = LinearInterpolate(A=Multiply(A=@n8, B=TextureSample(UVs=@n10, Tex=@n5) {samplerSource=SSM_Clamp_WorldGroupSettings}.A), Alpha=@n6.A)
  FunctionOutput(None=AppendVector(A=@n7, B=@n11)) {outputName=RGBA, sortPriority=5}
  FunctionOutput(None=ComponentMask(None=@n7) {r=true}) {outputName=R, sortPriority=1}
  FunctionOutput(None=ComponentMask(None=@n7) {g=true}) {outputName=G, sortPriority=2}
  FunctionOutput(None=ComponentMask(None=@n7) {b=true}) {outputName=B, sortPriority=3}
  FunctionOutput(None=ComponentMask(None=@n11) {a=true}) {outputName=A, sortPriority=4}
  FunctionOutput(None=@n7) {outputName=RGB}
 ]
}
